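# Masked material for the Whisper vehicle wheel, usable on skeletal meshes and built on the shared MF_WheelParent_Simple function.
# Base color blends from the parent's output toward a Custom_Color parameter through a Rim_mask texture, then fades to black by the vertex color red channel.
# The modified attributes pass through MF_BaseVehicle_Damage and then MF_VehicleDeRez_Transition, layering in the shared vehicle damage and de-rez transition functions.
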Material /VehicleCosmetics/Wheels/Whisper/Materials/M_Wheel_Whisper
graph {
 properties {
  bAutomaticallySetUsageInEditor=false
  bUseMaterialAttributes=true
  bUsedWithMeshDeformer=true
  bUsedWithSkeletalMesh=true
  blendMode=BLEND_Masked
  pixelDepthOffsetMode=PDOM_Legacy
  usedShadingModels=Default Lit
 }
 nodes [
  n1 = GetMaterialAttributes(None=MaterialFunctionCall() {materialFunction=&MF_WheelParent_Simple}.MaterialAttributes) {attributeGetTypes [#1]}
  n2 = LinearInterpolate(
   A=LinearInterpolate(
    A=@n1.Base Color
    B=VectorParameter() {parameterName=Custom_Color, defaultValue=<rgba 0 0 0 1>}.RGB
    Alpha=TextureSampleParameter2D() {parameterName=Rim_mask, texture=/VehicleCosmetics/Textures/T_Black}.RGB
   )
   Alpha=VertexColor().R
  ) {constB=0}
  MaterialFunctionCall(In=MaterialFunctionCall(Material=SetMaterialAttributes!(MaterialAttributes=@n1.MaterialAttributes, Base Color=@n2) {attributeSetTypes [#1]}) {materialFunction=&MF_BaseVehicle_Damage}.Result) {materialFunction=&MF_VehicleDeRez_Transition}
 ]
}
